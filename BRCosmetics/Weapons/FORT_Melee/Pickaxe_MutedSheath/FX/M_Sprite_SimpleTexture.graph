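# Translucent, unlit, two-sided Niagara sprite material for the MutedSheath pickaxe effects, showing a swappable texture that defaults to a sticker image.
# Texture RGB is multiplied by particle color and passed through the MF_FX_EmissiveAdjustments function as emissive.
# Texture alpha is multiplied by particle alpha, so the particle system controls both tint and transparency.
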Material /BRCosmetics/Weapons/FORT_Melee/Pickaxe_MutedSheath/FX/M_Sprite_SimpleTexture
graph {
 properties {
  bAutomaticallySetUsageInEditor=false
  bUsedWithNiagaraSprites=true
  blendMode=BLEND_Translucent
  shadingModel=MSM_Unlit
  twoSided=true
  usedShadingModels=Unlit
 }
 nodes [
  n1 = TextureSampleParameter2D() {parameterName=Texture, texture=&/T_Pickaxe_MutedSheath_Sticker01}
  n2 = ParticleColor()
  MaterialFunctionCall(Emissive=Multiply(A=@n1.RGB, B=@n2.RGB)) {materialFunction=&MF_FX_EmissiveAdjustments}
  Multiply(A=@n1.A, B=@n2.A)
 ]
}
